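; Fireball projectile for the Athena FrostedTips trap, extending a parent Curie projectile that spawns a building gameplay actor (BGA).
; On stop it fades out its looping audio and plays an ember impact sound unless it hit a water body.
; The SpawnCurieBGA event spawns the BGA at the line-trace transform.
; The server then hides the projectile when that BGA dies, while non-servers deactivate the fire trail right after spawning.
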(event EventBeginPlay
  (|CallParentFunction))

(event Collision|EventHit
  (params (MyComp :type Primitive%20Component%20Object%20Reference) (Other :type Actor%20Object%20Reference) (OtherComp :type Primitive%20Component%20Object%20Reference) (bSelfMoved :type Boolean) (HitLocation :type Vector :default "0, 0, 0") (HitNormal :type Vector :default "0, 0, 0") (NormalImpulse :type Vector :default "0, 0, 0") (Hit :type Hit%20Result%20Structure%20%28by%20ref%29))
  (|CallParentFunction MyComp Other OtherComp bSelfMoved HitLocation HitNormal NormalImpulse Hit))

(event EventOnStop
  (params (Hit :type Hit%20Result%20Structure%20%28by%20ref%29))
  (bind (_bblockinghit _binitialoverlap _time _distance _location _impactpoint _normal _impactnormal _physmat _hitactor) (Collision|BreakHitResult Hit))
  (|CallParentFunction Hit)
  (Audio|Components|Audio|FadeOut (Variables|Default|GetLoopingAudioTell) 0.25 0.0 "Linear" :!class /Script/Engine.AudioComponent)
  (Utilities|Casting|CastToWaterBody _hitactor
    (:CastFailed
      (CallFunction|PlayImpactSound))))

(event Custom|Play%20Impact%20Sound
  (Audio|PlaySoundatLocation "/Game/Sounds/Fort_Weapons/pistol/FlareGun/FlareGun_EmberImpact_Cue.FlareGun_EmberImpact_Cue" (.ImpactPoint (Collision|BreakHitResult (stale EventOnStop Hit)))))

(event EventSpawnCurieBGA
  ;[1 1 1] Spawn curie bga and handle shorter ignite time
  (bind _returnvalue (Building|SpawnBuildingGameplayActor (Variables|CurieProjectile|GetBGAToSpawn) (Variables|Default|GetLineTraceTransform) (Actor|GetOwner)))
  (if (Networking|IsServer)
    (EventDispatchers|BindEventtoOnDied _returnvalue (EventDispatchers|CreateEvent :!event Custom|OnDied) :!class /Script/FortniteGame.BuildingActor)
    (else
      (Components|Activation|Deactivate (Variables|Default|GetFireTrail)))))

(event Custom|OnDied
  (params (DamagedActor :type Actor%20Object%20Reference) (Damage :type Float%20%28single-precision%29) (InstigatedBy :type Controller%20Object%20Reference) (DamageCauser :type Actor%20Object%20Reference) (HitLocation :type Vector) (FHitComponent :type Primitive%20Component%20Object%20Reference) (BoneName :type Name) (Momentum :type Vector))
  ;[1 1 1] Spawn curie bga and handle shorter ignite time
  (Rendering|SetActorHiddenInGame true))

(fn ConstructionScript)
