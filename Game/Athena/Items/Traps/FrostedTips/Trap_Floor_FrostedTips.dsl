(event BuildingActor|EventPlacedByPlayer
  (params (PlayerState :type Fort%20Player%20State%20Zone%20Object%20Reference))
  ;[1 1 1] Set Owner to player state so trap owner gets curie kill credit
  (Actor|SetOwner PlayerState))

(event EventOnPlaced
  ;[1 1 1] Initial Setup
  (Trap|AddTriggerComponent (Variables|Default|GetTriggerVolume))
  (Audio|PlaySoundatLocation (Variables|Audio|GetTrap_Placed_Sound) (Transformation|GetWorldLocation (Variables|Default|GetPlacementSoundLocation)))
  (Utilities|TrapPlacementDustFX :Attach%20to%20Component (Variables|Debug|GetStaticMeshComponent) :Socket "PlacedFX" :Rotation "0, 0, 0" :Burst%20Rate 8 :Size%20%28Volume%29 "100.000000,100.000000,15.000000" :Velocity%20Scale "1,1,1" :Size%20Scale%20%28Particles%29 1
    (:Out
      (bind _returnvalue (CallFunction|GetorCreateMIDOnDefaultMesh 0))
      (Variables|Default|SetTrapMID _returnvalue))))

(event EventOnFinishedBuilding
  (bind _returnvalue (Transformation|GetActorLocation))
  ;[1 1 1] Finished building
  (Audio|PlaySoundatLocation (Variables|Audio|GetTrap_Active_Sound) _returnvalue)
  (Components|Activation|Activate (Variables|Default|GetIdleFX) false)
  (Rendering|Material|SetVectorParameterValue (Variables|Default|GetTrapMID) "EmissiveMultiplier" (Variables|Default|GetActiveEmissiveColor) :!class /Script/Engine.MaterialInstanceDynamic)
  (Components|Activation|Activate (Variables|Default|GetP_FlameGrill_Trap_Start) false)
  (Audio|PlaySoundatLocation (Variables|Audio|GetTrap_GasIgnite_Sound) _returnvalue)
  (Development|SetHiddeninGame (Variables|Default|GetSM_Flames_Pilot_Light_Group_01) false false)
  (bind (_direction _reveal _grow _growradial _opacity _colortrack) (Timeline|Timeline_0 PlayFromStart)
    (:Update
      (Rendering|Material|SetScalarParameterValueonMaterials (Variables|Default|GetSM_Flames_Pilot_Light_Group_01) "Reveal" _reveal)
      (Rendering|Material|SetVectorParameterValueonMaterials (Variables|Debug|GetStaticMeshComponent) "EmissiveMultiplier" (Math|Conversions|ToVector(LinearColor) _colortrack)))))

(event GameplayCue|GameplayCue.Abilities.Activation.Traps.DelayBegin
  (params (EventType :type EGameplayCueEvent%20Enum :default "OnActive") (Parameters :type Gameplay%20Cue%20Parameters%20Structure%20%28by%20ref%29))
  ;[1 1 1] On Fire
  (switch Utilities|FlowControl|Switch|SwitchonEGameplayCueEvent EventType
    (:Executed
      (Audio|PlaySoundatLocation (Variables|Audio|GetTrap_PreFire_Sound) (Transformation|GetActorLocation)))))

(event GameplayCue|GameplayCue.Abilities.Activation.Traps.ActivateTrap
  (params (EventType :type EGameplayCueEvent%20Enum :default "OnActive") (Parameters :type Gameplay%20Cue%20Parameters%20Structure%20%28by%20ref%29))
  (bind _sm_flames_pilot_light_group_01 (Variables|Default|GetSM_Flames_Pilot_Light_Group_01))
  ;[1 1 1] On Fire
  (switch Utilities|FlowControl|Switch|SwitchonEGameplayCueEvent EventType
    (:Executed
      (Audio|PlaySoundatLocation (Variables|Audio|GetTrap_Fire_Sound) (Transformation|GetActorLocation) "0, 0, 0" 1.5 0.75)
      (CallFunction|SetActiveFireFX true)
      (Components|Activation|SetActive (Variables|Default|GetIdleFX) false false)
      (Rendering|Material|SetScalarParameterValueonMaterials _sm_flames_pilot_light_group_01 "Reveal" -1.0)
      (Development|SetHiddeninGame _sm_flames_pilot_light_group_01 true false)
      (bind (_direction _emissivetrack) (Timeline|Timeline_1 PlayFromStart)
        (:Update
          (Rendering|Material|SetVectorParameterValueonMaterials (Variables|Debug|GetStaticMeshComponent) "EmissiveMultiplier" (Math|Conversions|ToVector(LinearColor) _emissivetrack)))))))

(event GameplayCue|GameplayCue.Abilities.Activation.Traps.ReloadBegin
  (params (EventType :type EGameplayCueEvent%20Enum :default "OnActive") (Parameters :type Gameplay%20Cue%20Parameters%20Structure%20%28by%20ref%29))
  ;[1 1 1] Reload Begin
  (switch Utilities|FlowControl|Switch|SwitchonEGameplayCueEvent EventType
    (:Executed
      (Audio|PlaySoundatLocation (Variables|Audio|GetTrap_Reload_Sound) (Transformation|GetActorLocation)))))

(event EventOnReloadEnd
  ;[1 1 1] Reload End
  (Audio|PlaySoundatLocation (Variables|Audio|GetTrap_Active_Sound) (Transformation|GetActorLocation))
  (Components|Activation|SetActive (Variables|Default|GetIdleFX) true false)
  (Rendering|Material|SetVectorParameterValue (Variables|Default|GetTrapMID) "EmissiveMultiplier" (Variables|Default|GetActiveEmissiveColor) :!class /Script/Engine.MaterialInstanceDynamic)
  (Components|Activation|Activate (Variables|Default|GetP_FlameGrill_Trap_Start) false)
  (Development|SetHiddeninGame (Variables|Default|GetSM_Flames_Pilot_Light_Group_01) false false)
  (bind (_direction _reveal _grow _growradial _opacity _colortrack) (Timeline|Timeline_2 PlayFromStart)
    (:Update
      (Rendering|Material|SetScalarParameterValueonMaterials (Variables|Default|GetSM_Flames_Pilot_Light_Group_01) "Reveal" _reveal)
      (Rendering|Material|SetVectorParameterValueonMaterials (Variables|Debug|GetStaticMeshComponent) "EmissiveMultiplier" (Math|Conversions|ToVector(LinearColor) _colortrack)))))

(event EventOnOutOfDurability
  ;[1 1 1] Durability Change
  (Utilities|FlowControl|Delay 1)
  (Audio|PlaySoundatLocation (Variables|Audio|GetTrap_Explode_Sound) (Transformation|GetActorLocation))
  (Effects|SpawnEmitteratLocation "/Game/Weapons/FORT_TRAPS/Spike_Trap/P_Spike_Trap_Destruction.P_Spike_Trap_Destruction" (Transformation|GetWorldLocation (Variables|Transformation|GetRootComponent))))

(event EventOnDeathServer
  (params (Damage :type Float%20%28single-precision%29) (DamageTags :type Gameplay%20Tag%20Container%20Structure%20%28by%20ref%29) (Momentum :type Vector :default "0, 0, 0") (HitInfo :type Hit%20Result%20Structure%20%28by%20ref%29) (InstigatedBy :type Controller%20Object%20Reference) (DamageCauser :type Actor%20Object%20Reference) (EffectContext :type Gameplay%20Effect%20Context%20Handle%20Structure))
  (Rendering|SetActorHiddenInGame true))

(fn SetActiveFireFX
  (params (_Active :type Boolean))
  (for _array_element (Variables|Default|GetActiveFXEmitters)
    (Components|Activation|SetActive _array_element _Active false)))

(fn ConstructionScript
  (Variables|Default|SetActiveFXEmitters (Utilities|Array|MakeArray (Variables|Default|GetActiveFX) (Variables|Default|GetActiveFX_1) (Variables|Default|GetActiveFX_2) (Variables|Default|GetActiveFX_3)))
  (Variables|Default|SetFlameSpawnersSpots (Utilities|Array|MakeArray (Variables|Default|GetFireMuzzle) (Variables|Default|GetFireMuzzle1) (Variables|Default|GetFireMuzzle4) (Variables|Default|GetFireMuzzle2) (Variables|Default|GetFireMuzzle3))))
